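# Physics asset for the left and right hoodie arm flaps (tassels) on the M_MED_Glow_Bro player body.
# Each flap is a mirrored box body constrained with narrow swing and twist limits to a tiny capsule on its MSTR bone, which serves as an anchor.
# Both anchors link to a spine_05 torso capsule, which links to a pelvis capsule, all with free angular motion.
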PhysicsAsset /BRCosmetics/Characters/Player/Male/Medium/Bodies/M_MED_Glow_Bro/Meshes/M_MED_Glow_Bro_ArmTassles_Physics
bodies {
 dyn_hoodie_armFlap_01_l [{shapeType=Box, extentX=4.3334, extentY=4.9417, extentZ=1.0681, center=<2.3888 0 0>}]
 dyn_hoodie_armFlap_01_l_MSTR [{shapeType=Capsule, length=0.01, radius=0.01}]
 dyn_hoodie_armFlap_01_r [{shapeType=Box, extentX=4.3334, extentY=4.9417, extentZ=1.0681, center=<-2.3888 0 0>, rotation=<rot 0 0 180>}]
 dyn_hoodie_armFlap_01_r_MSTR [{shapeType=Capsule, length=0.01, radius=0.01}]
 pelvis [{shapeType=Capsule, length=10.8441, radius=10.2509}]
 spine_05 [{shapeType=Capsule, length=31.5, radius=15.0592, center=<-3.9916 -2.3084 -0.0006>, rotation=<rot -0.0003 -0.0088 0>}]
}
constraints [
 {
  bone1Name=dyn_hoodie_armFlap_01_l
  bone2Name=dyn_hoodie_armFlap_01_l_MSTR
  swing1LimitDegrees=25
  swing2LimitDegrees=25
  twistLimitDegrees=15
 }
 {
  bone1Name=dyn_hoodie_armFlap_01_l_MSTR
  bone2Name=spine_05
  swing1LimitDegrees=45
  swing1Motion=Free
  swing2LimitDegrees=45
  swing2Motion=Free
  twistLimitDegrees=45
  twistMotion=Free
 }
 {
  bone1Name=dyn_hoodie_armFlap_01_r_MSTR
  bone2Name=spine_05
  swing1LimitDegrees=45
  swing1Motion=Free
  swing2LimitDegrees=45
  swing2Motion=Free
  twistLimitDegrees=45
  twistMotion=Free
 }
 {
  bone1Name=dyn_hoodie_armFlap_01_r
  bone2Name=dyn_hoodie_armFlap_01_r_MSTR
  swing1LimitDegrees=25
  swing2LimitDegrees=25
  twistLimitDegrees=15
 }
 {
  bone1Name=spine_05
  bone2Name=pelvis
  swing1LimitDegrees=45
  swing1Motion=Free
  swing2LimitDegrees=45
  swing2Motion=Free
  twistLimitDegrees=45
  twistMotion=Free
 }
]
