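# Physics asset for the coat tassels of the Glow Bro male medium body, giving two hoodie cords bodies for physics simulation.
# Each cord is a five-box chain with swing and tighter twist limits, rooted on a tiny master capsule linked to spine_05.
# Pelvis and spine capsules form the torso, joined by fully free angular constraints, likely as collision for the cords.
# Spine_04 also carries a second, thinner capsule set far off its bone origin.
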PhysicsAsset /BRCosmetics/Characters/Player/Male/Medium/Bodies/M_MED_Glow_Bro/Meshes/M_MED_Glow_Bro_CoatTassles_Physics
bodies {
 dyn_hoodie_cord_01_l [{shapeType=Box, extentX=5.0561, extentY=1.4384, extentZ=3.2393, center=<2.8831 -0.0001 0>, rotation=<rot 0 0 24.9996>}]
 dyn_hoodie_cord_01_l_MSTR [{shapeType=Capsule, length=0.01, radius=0.01}]
 dyn_hoodie_cord_01_r [{shapeType=Box, extentX=5.0561, extentY=1.4384, extentZ=3.2393, center=<-2.8831 0.0001 0>, rotation=<rot 0 0 -155.0004>}]
 dyn_hoodie_cord_01_r_MSTR [{shapeType=Capsule, length=0.01, radius=0.01}]
 dyn_hoodie_cord_02_l [{shapeType=Box, extentX=5.0561, extentY=1.4384, extentZ=3.2393, center=<2.8831 -0.0001 0>, rotation=<rot 0 0 24.9996>}]
 dyn_hoodie_cord_02_r [{shapeType=Box, extentX=5.0561, extentY=1.4384, extentZ=3.2393, center=<-2.8831 0.0001 0>, rotation=<rot 0 0 -155.0004>}]
 dyn_hoodie_cord_03_l [{shapeType=Box, extentX=5.0561, extentY=1.4384, extentZ=3.2393, center=<2.8831 -0.0001 0>, rotation=<rot 0 0 24.9996>}]
 dyn_hoodie_cord_03_r [{shapeType=Box, extentX=5.0561, extentY=1.4384, extentZ=3.2393, center=<-2.8831 0.0001 0>, rotation=<rot 0 0 -155.0004>}]
 dyn_hoodie_cord_04_l [{shapeType=Box, extentX=5.0561, extentY=1.4384, extentZ=3.2393, center=<2.8831 -0.0001 0>, rotation=<rot 0 0 24.9996>}]
 dyn_hoodie_cord_04_r [{shapeType=Box, extentX=5.0561, extentY=1.4384, extentZ=3.2393, center=<-2.8831 0.0001 0>, rotation=<rot 0 0 -155.0004>}]
 dyn_hoodie_cord_05_l [{shapeType=Box, extentX=5.0561, extentY=1.4384, extentZ=3.2393, center=<2.8831 -0.0001 0>, rotation=<rot 0 0 24.9996>}]
 dyn_hoodie_cord_05_r [{shapeType=Box, extentX=5.0561, extentY=1.4384, extentZ=3.2393, center=<-2.8831 0.0001 0>, rotation=<rot 0 0 -155.0004>}]
 pelvis [{shapeType=Capsule, length=10.8441, radius=10.2509}]
 spine_01 [{shapeType=Capsule, length=10.3441, radius=10.0009, center=<0 -0.2248 0>, rotation=<rot 0 -10.0004 0>}]
 spine_02 [{shapeType=Capsule, length=8.7144, radius=10.0893, center=<0 0.1454 0>, rotation=<rot 0 -90 0>}]
 spine_04 [
  {shapeType=Capsule, length=30, radius=15.5592, center=<-0.4941 -2.0935 -0.0007>, rotation=<rot -0.0003 -10.0068 0.0001>}
  {shapeType=Capsule, length=32.5, radius=5.5, center=<26.6555 151.1288 0.0012>, rotation=<rot 0.0001 79.9968 89.9996>}
 ]
 spine_05 [{shapeType=Capsule, length=31.5, radius=15.0592, center=<-3.9916 -2.3084 -0.0006>, rotation=<rot -0.0003 -0.0088 0>}]
}
constraints [
 {
  bone1Name=spine_04
  bone2Name=spine_02
  swing1LimitDegrees=45
  swing1Motion=Free
  swing2LimitDegrees=45
  swing2Motion=Free
  twistLimitDegrees=45
  twistMotion=Free
 }
 {
  bone1Name=dyn_hoodie_cord_01_r_MSTR
  bone2Name=spine_05
  swing1LimitDegrees=45
  swing1Motion=Free
  swing2LimitDegrees=45
  swing2Motion=Free
  twistLimitDegrees=45
  twistMotion=Free
 }
 {
  bone1Name=dyn_hoodie_cord_01_l_MSTR
  bone2Name=spine_05
  swing1LimitDegrees=45
  swing1Motion=Free
  swing2LimitDegrees=45
  swing2Motion=Free
  twistLimitDegrees=45
  twistMotion=Free
 }
 {
  bone1Name=spine_02
  bone2Name=spine_01
  swing1LimitDegrees=45
  swing1Motion=Free
  swing2LimitDegrees=45
  swing2Motion=Free
  twistLimitDegrees=45
  twistMotion=Free
 }
 {
  bone1Name=spine_01
  bone2Name=pelvis
  swing1LimitDegrees=45
  swing1Motion=Free
  swing2LimitDegrees=45
  swing2Motion=Free
  twistLimitDegrees=45
  twistMotion=Free
 }
 {bone1Name=dyn_hoodie_cord_01_l, bone2Name=dyn_hoodie_cord_01_l_MSTR, swing1LimitDegrees=45, swing2LimitDegrees=45, twistLimitDegrees=15}
 {bone1Name=dyn_hoodie_cord_02_l, bone2Name=dyn_hoodie_cord_01_l, swing1LimitDegrees=45, swing2LimitDegrees=45, twistLimitDegrees=15}
 {bone1Name=dyn_hoodie_cord_03_l, bone2Name=dyn_hoodie_cord_02_l, swing1LimitDegrees=45, swing2LimitDegrees=45, twistLimitDegrees=15}
 {bone1Name=dyn_hoodie_cord_04_l, bone2Name=dyn_hoodie_cord_03_l, swing1LimitDegrees=45, swing2LimitDegrees=45, twistLimitDegrees=5}
 {bone1Name=dyn_hoodie_cord_05_l, bone2Name=dyn_hoodie_cord_04_l, swing1LimitDegrees=45, swing2LimitDegrees=45, twistLimitDegrees=15}
 {
  bone1Name=spine_05
  bone2Name=spine_04
  swing1LimitDegrees=45
  swing1Motion=Free
  swing2LimitDegrees=45
  swing2Motion=Free
  twistLimitDegrees=45
  twistMotion=Free
 }
 {bone1Name=dyn_hoodie_cord_01_r, bone2Name=dyn_hoodie_cord_01_r_MSTR, swing1LimitDegrees=45, swing2LimitDegrees=45, twistLimitDegrees=15}
 {bone1Name=dyn_hoodie_cord_02_r, bone2Name=dyn_hoodie_cord_01_r, swing1LimitDegrees=45, swing2LimitDegrees=45, twistLimitDegrees=5}
 {bone1Name=dyn_hoodie_cord_03_r, bone2Name=dyn_hoodie_cord_02_r, swing1LimitDegrees=45, swing2LimitDegrees=45, twistLimitDegrees=15}
 {bone1Name=dyn_hoodie_cord_04_r, bone2Name=dyn_hoodie_cord_03_r, swing1LimitDegrees=45, swing2LimitDegrees=45, twistLimitDegrees=5}
 {bone1Name=dyn_hoodie_cord_05_r, bone2Name=dyn_hoodie_cord_04_r, swing1LimitDegrees=45, swing2LimitDegrees=45, twistLimitDegrees=15}
]
