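# Physics asset for the M_MED_Glow_Bro male player body cosmetic, set up for secondary motion on its hoodie, jacket and shirt.
# Capsules cover torso and legs, while boxes on dyn_ bones form hoodie, zipper, two-link jacket and rear shirt pieces.
# Tiny MSTR capsules join to spine, pelvis or thighs with free rotation, anchoring the angle-limited cloth boxes.
# Hair, ponytail, front shirt and hoodie_02 have only MSTR anchors, without child bodies.
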PhysicsAsset /BRCosmetics/Characters/Player/Male/Medium/Bodies/M_MED_Glow_Bro/Meshes/M_MED_Glow_Bro_Physics
bodies {
 calf_twist_02_l [{shapeType=Capsule, length=26.9405, radius=10.1346, center=<0 0.0001 2.3018>, rotation=<rot 90 0 0>}]
 calf_twist_02_r [{shapeType=Capsule, length=26.9405, radius=10.1346, center=<0 -0.0001 -2.3018>, rotation=<rot 90 0 180>}]
 dyn_hairtop_01_MSTR [{shapeType=Capsule, length=0.01, radius=0.01}]
 dyn_hairtop_02_MSTR [{shapeType=Capsule, length=0.01, radius=0.01}]
 dyn_hairtop_03_MSTR [{shapeType=Capsule, length=0.01, radius=0.01}]
 dyn_hairtop_05_MSTR [{shapeType=Capsule, length=0.01, radius=0.01}]
 dyn_hairtop_06_MSTR [{shapeType=Capsule, length=0.01, radius=0.01}]
 dyn_hairtop_07_MSTR [{shapeType=Capsule, length=0.01, radius=0.01}]
 dyn_hoodie_01 [{shapeType=Box, extentX=4.8334, extentY=4.4417, extentZ=1.5681, center=<0.1181 0 -0.0001>}]
 dyn_hoodie_01_MSTR [{shapeType=Capsule, length=0.01, radius=0.01}]
 dyn_hoodie_01_l [{shapeType=Box, extentX=4.8334, extentY=4.4417, extentZ=1.5681, center=<0.1181 0 -0.0001>}]
 dyn_hoodie_01_l_MSTR [{shapeType=Capsule, length=0.01, radius=0.01}]
 dyn_hoodie_01_r [{shapeType=Box, extentX=4.8334, extentY=4.4417, extentZ=1.5681, center=<-0.1181 0 0.0001>, rotation=<rot 0 0 180>}]
 dyn_hoodie_01_r_MSTR [{shapeType=Capsule, length=0.01, radius=0.01}]
 dyn_hoodie_02_l_MSTR [{shapeType=Capsule, length=0.01, radius=0.01}]
 dyn_hoodie_02_r_MSTR [{shapeType=Capsule, length=0.01, radius=0.01}]
 dyn_hoodie_zipper_01_l [{shapeType=Box, extentX=4.5561, extentY=3.4384, extentZ=1.1, center=<2.0114 0 0>}]
 dyn_hoodie_zipper_01_r [{shapeType=Box, extentX=4.5561, extentY=3.4384, extentZ=1.1, center=<-2.0114 0 0>, rotation=<rot 0 0 180>}]
 dyn_jacket_back_01_l [{shapeType=Box, extentX=5.5561, extentY=9.9384, extentZ=2.1, center=<2.6572 0 0>}]
 dyn_jacket_back_01_l_MSTR [{shapeType=Capsule, length=0.01, radius=0.01}]
 dyn_jacket_back_01_r [{shapeType=Box, extentX=5.5561, extentY=9.9384, extentZ=2.1, center=<-2.6572 0 0>, rotation=<rot 0 0 180>}]
 dyn_jacket_back_01_r_MSTR [{shapeType=Capsule, length=0.01, radius=0.01}]
 dyn_jacket_back_02_l [{shapeType=Box, extentX=5.5561, extentY=9.9384, extentZ=2.1, center=<2.6572 0 0>}]
 dyn_jacket_back_02_r [{shapeType=Box, extentX=5.5561, extentY=9.9384, extentZ=2.1, center=<-2.6572 0 0>, rotation=<rot 0 0 180>}]
 dyn_jacket_front_01_l [{shapeType=Box, extentX=5.5561, extentY=9.9384, extentZ=1.1, center=<2.6572 0 0>}]
 dyn_jacket_front_01_l_MSTR [{shapeType=Capsule, length=0.01, radius=0.01}]
 dyn_jacket_front_01_r [{shapeType=Box, extentX=5.5561, extentY=9.9384, extentZ=1.1, center=<-2.6572 0 0>, rotation=<rot 0 0 180>}]
 dyn_jacket_front_01_r_MSTR [{shapeType=Capsule, length=0.01, radius=0.01}]
 dyn_jacket_front_02_l [{shapeType=Box, extentX=5.5561, extentY=9.9384, extentZ=1.1, center=<2.6572 0 0>}]
 dyn_jacket_front_02_r [{shapeType=Box, extentX=5.5561, extentY=9.9384, extentZ=1.1, center=<-2.6572 0 0>, rotation=<rot 0 0 180>}]
 dyn_jacket_side_01_l [{shapeType=Box, extentX=5.5561, extentY=9.9384, extentZ=2.1, center=<2.6572 0 0>}]
 dyn_jacket_side_01_l_MSTR [{shapeType=Capsule, length=0.01, radius=0.01}]
 dyn_jacket_side_01_r [{shapeType=Box, extentX=5.5561, extentY=9.9384, extentZ=2.1, center=<-2.6572 0 0>, rotation=<rot 0 0 180>}]
 dyn_jacket_side_01_r_MSTR [{shapeType=Capsule, length=0.01, radius=0.01}]
 dyn_jacket_side_02_l [{shapeType=Box, extentX=5.5561, extentY=9.9384, extentZ=2.1, center=<2.6572 0 0>}]
 dyn_jacket_side_02_r [{shapeType=Box, extentX=5.5561, extentY=9.9384, extentZ=2.1, center=<-2.6572 0 0>, rotation=<rot 0 0 180>}]
 dyn_ponytail_01_MSTR [{shapeType=Capsule, length=0.01, radius=0.01}]
 dyn_shirt_back_01_l [{shapeType=Box, extentX=13.0561, extentY=5.9384, extentZ=2, center=<6.7978 0 0>}]
 dyn_shirt_back_01_l_MSTR [{shapeType=Capsule, length=0.01, radius=0.01}]
 dyn_shirt_back_01_r [{shapeType=Box, extentX=13.0561, extentY=5.9384, extentZ=2, center=<-6.7978 0 0>, rotation=<rot 0 0 180>}]
 dyn_shirt_back_01_r_MSTR [{shapeType=Capsule, length=0.01, radius=0.01}]
 dyn_shirt_back_center_01 [{shapeType=Box, extentX=2, extentY=5.9384, extentZ=13.0561, center=<0 0 -6.7978>}]
 dyn_shirt_back_center_01_MSTR [{shapeType=Capsule, length=0.01, radius=0.01}]
 dyn_shirt_front_01_l_MSTR [{shapeType=Capsule, length=0.01, radius=0.01}]
 dyn_shirt_front_01_r_MSTR [{shapeType=Capsule, length=0.01, radius=0.01}]
 pelvis [{shapeType=Capsule, length=10.8441, radius=10.2509}]
 spine_01 [{shapeType=Capsule, length=10.3441, radius=10.0009, center=<0 -0.2248 0>, rotation=<rot 0 -10.0004 0>}]
 spine_02 [{shapeType=Capsule, length=8.7144, radius=10.0893, center=<0 0.1454 0>, rotation=<rot 0 -90 0>}]
 spine_04 [
  {shapeType=Capsule, length=30, radius=15.5592, center=<-0.4941 -2.0935 -0.0007>, rotation=<rot -0.0003 -10.0068 0.0001>}
  {shapeType=Capsule, length=32.5, radius=5.5, center=<26.6555 151.1288 0.0012>, rotation=<rot 0.0001 79.9968 89.9996>}
 ]
 spine_05 [{shapeType=Capsule, length=31.5, radius=15.0592, center=<-3.9916 -2.3084 -0.0006>, rotation=<rot -0.0003 -0.0088 0>}]
 thigh_l [{shapeType=Capsule, length=3, radius=9.3781, rotation=<rot 0 0 -90>}]
 thigh_r [{shapeType=Capsule, length=3, radius=9.3781, rotation=<rot 0 0 90>}]
 thigh_twist_01_l [{shapeType=Capsule, length=29.9405, radius=9.1346, rotation=<rot 90 0 0>}]
 thigh_twist_01_r [{shapeType=Capsule, length=29.9405, radius=9.1346, rotation=<rot 90 0 180>}]
}
constraints [
 {bone1Name=thigh_twist_01_l, bone2Name=thigh_l, swing1LimitDegrees=45, swing2LimitDegrees=45, twistLimitDegrees=45}
 {bone1Name=thigh_twist_01_r, bone2Name=thigh_r, swing1LimitDegrees=45, swing2LimitDegrees=45, twistLimitDegrees=45}
 {bone1Name=dyn_hoodie_01, bone2Name=dyn_hoodie_01_MSTR, swing1LimitDegrees=5, swing2LimitDegrees=15, twistLimitDegrees=5}
 {bone1Name=dyn_hoodie_01_l, bone2Name=dyn_hoodie_01_l_MSTR, swing1LimitDegrees=5, swing2LimitDegrees=15, twistLimitDegrees=5}
 {
  bone1Name=spine_04
  bone2Name=spine_02
  swing1LimitDegrees=45
  swing1Motion=Free
  swing2LimitDegrees=45
  swing2Motion=Free
  twistLimitDegrees=45
  twistMotion=Free
 }
 {
  bone1Name=dyn_ponytail_01_MSTR
  bone2Name=spine_05
  swing1LimitDegrees=45
  swing1Motion=Free
  swing2LimitDegrees=45
  swing2Motion=Free
  twistLimitDegrees=45
  twistMotion=Free
 }
 {
  bone1Name=dyn_hairtop_02_MSTR
  bone2Name=spine_05
  swing1LimitDegrees=45
  swing1Motion=Free
  swing2LimitDegrees=45
  swing2Motion=Free
  twistLimitDegrees=45
  twistMotion=Free
 }
 {
  bone1Name=dyn_hairtop_01_MSTR
  bone2Name=spine_05
  swing1LimitDegrees=45
  swing1Motion=Free
  swing2LimitDegrees=45
  swing2Motion=Free
  twistLimitDegrees=45
  twistMotion=Free
 }
 {
  bone1Name=dyn_hairtop_06_MSTR
  bone2Name=spine_05
  swing1LimitDegrees=45
  swing1Motion=Free
  swing2LimitDegrees=45
  swing2Motion=Free
  twistLimitDegrees=45
  twistMotion=Free
 }
 {
  bone1Name=dyn_hairtop_07_MSTR
  bone2Name=spine_05
  swing1LimitDegrees=45
  swing1Motion=Free
  swing2LimitDegrees=45
  swing2Motion=Free
  twistLimitDegrees=45
  twistMotion=Free
 }
 {
  bone1Name=dyn_hairtop_03_MSTR
  bone2Name=spine_05
  swing1LimitDegrees=45
  swing1Motion=Free
  swing2LimitDegrees=45
  swing2Motion=Free
  twistLimitDegrees=45
  twistMotion=Free
 }
 {
  bone1Name=dyn_hairtop_05_MSTR
  bone2Name=spine_05
  swing1LimitDegrees=45
  swing1Motion=Free
  swing2LimitDegrees=45
  swing2Motion=Free
  twistLimitDegrees=45
  twistMotion=Free
 }
 {
  bone1Name=dyn_hoodie_02_r_MSTR
  bone2Name=spine_05
  swing1LimitDegrees=45
  swing1Motion=Free
  swing2LimitDegrees=45
  swing2Motion=Free
  twistLimitDegrees=45
  twistMotion=Free
 }
 {
  bone1Name=dyn_hoodie_01_r_MSTR
  bone2Name=spine_05
  swing1LimitDegrees=45
  swing1Motion=Free
  swing2LimitDegrees=45
  swing2Motion=Free
  twistLimitDegrees=45
  twistMotion=Free
 }
 {
  bone1Name=dyn_hoodie_01_MSTR
  bone2Name=spine_05
  swing1LimitDegrees=45
  swing1Motion=Free
  swing2LimitDegrees=45
  swing2Motion=Free
  twistLimitDegrees=45
  twistMotion=Free
 }
 {
  bone1Name=dyn_hoodie_01_l_MSTR
  bone2Name=spine_05
  swing1LimitDegrees=45
  swing1Motion=Free
  swing2LimitDegrees=45
  swing2Motion=Free
  twistLimitDegrees=45
  twistMotion=Free
 }
 {
  bone1Name=dyn_hoodie_02_l_MSTR
  bone2Name=spine_05
  swing1LimitDegrees=45
  swing1Motion=Free
  swing2LimitDegrees=45
  swing2Motion=Free
  twistLimitDegrees=45
  twistMotion=Free
 }
 {
  bone1Name=dyn_jacket_front_01_l_MSTR
  bone2Name=spine_02
  swing1LimitDegrees=45
  swing1Motion=Free
  swing2LimitDegrees=45
  swing2Motion=Free
  twistLimitDegrees=45
  twistMotion=Free
 }
 {
  bone1Name=dyn_jacket_side_01_l_MSTR
  bone2Name=spine_02
  swing1LimitDegrees=45
  swing1Motion=Free
  swing2LimitDegrees=45
  swing2Motion=Free
  twistLimitDegrees=45
  twistMotion=Free
 }
 {
  bone1Name=dyn_jacket_back_01_l_MSTR
  bone2Name=spine_02
  swing1LimitDegrees=45
  swing1Motion=Free
  swing2LimitDegrees=45
  swing2Motion=Free
  twistLimitDegrees=45
  twistMotion=Free
 }
 {
  bone1Name=dyn_jacket_back_01_r_MSTR
  bone2Name=spine_02
  swing1LimitDegrees=45
  swing1Motion=Free
  swing2LimitDegrees=45
  swing2Motion=Free
  twistLimitDegrees=45
  twistMotion=Free
 }
 {
  bone1Name=dyn_jacket_side_01_r_MSTR
  bone2Name=spine_02
  swing1LimitDegrees=45
  swing1Motion=Free
  swing2LimitDegrees=45
  swing2Motion=Free
  twistLimitDegrees=45
  twistMotion=Free
 }
 {
  bone1Name=dyn_jacket_front_01_r_MSTR
  bone2Name=spine_02
  swing1LimitDegrees=45
  swing1Motion=Free
  swing2LimitDegrees=45
  swing2Motion=Free
  twistLimitDegrees=45
  twistMotion=Free
 }
 {
  bone1Name=dyn_shirt_front_01_l_MSTR
  bone2Name=thigh_twist_01_l
  swing1LimitDegrees=45
  swing1Motion=Free
  swing2LimitDegrees=45
  swing2Motion=Free
  twistLimitDegrees=45
  twistMotion=Free
 }
 {
  bone1Name=dyn_shirt_front_01_r_MSTR
  bone2Name=thigh_twist_01_r
  swing1LimitDegrees=45
  swing1Motion=Free
  swing2LimitDegrees=45
  swing2Motion=Free
  twistLimitDegrees=45
  twistMotion=Free
 }
 {
  bone1Name=dyn_shirt_back_01_l_MSTR
  bone2Name=pelvis
  swing1LimitDegrees=45
  swing1Motion=Free
  swing2LimitDegrees=45
  swing2Motion=Free
  twistLimitDegrees=45
  twistMotion=Free
 }
 {
  bone1Name=dyn_shirt_back_center_01_MSTR
  bone2Name=pelvis
  swing1LimitDegrees=45
  swing1Motion=Free
  swing2LimitDegrees=45
  swing2Motion=Free
  twistLimitDegrees=45
  twistMotion=Free
 }
 {
  bone1Name=dyn_shirt_back_01_r_MSTR
  bone2Name=pelvis
  swing1LimitDegrees=45
  swing1Motion=Free
  swing2LimitDegrees=45
  swing2Motion=Free
  twistLimitDegrees=45
  twistMotion=Free
 }
 {bone1Name=dyn_hoodie_zipper_01_l, bone2Name=dyn_jacket_front_01_l, swing1LimitDegrees=45, swing2LimitDegrees=45, twistLimitDegrees=45}
 {
  bone1Name=dyn_shirt_back_center_01
  bone2Name=dyn_shirt_back_center_01_MSTR
  swing1LimitDegrees=25
  swing2LimitDegrees=25
  twistLimitDegrees=5
 }
 {
  bone1Name=calf_twist_02_l
  bone2Name=thigh_l
  swing1LimitDegrees=45
  swing1Motion=Free
  swing2LimitDegrees=45
  swing2Motion=Free
  twistLimitDegrees=45
  twistMotion=Free
 }
 {
  bone1Name=calf_twist_02_r
  bone2Name=thigh_r
  swing1LimitDegrees=45
  swing1Motion=Free
  swing2LimitDegrees=45
  swing2Motion=Free
  twistLimitDegrees=45
  twistMotion=Free
 }
 {bone1Name=dyn_shirt_back_01_l, bone2Name=dyn_shirt_back_01_l_MSTR, swing1LimitDegrees=25, swing2LimitDegrees=25, twistLimitDegrees=5}
 {bone1Name=dyn_shirt_back_01_r, bone2Name=dyn_shirt_back_01_r_MSTR, swing1LimitDegrees=25, swing2LimitDegrees=25, twistLimitDegrees=5}
 {bone1Name=dyn_jacket_back_01_l, bone2Name=dyn_jacket_back_01_l_MSTR, swing1LimitDegrees=15, swing2LimitDegrees=45, twistLimitDegrees=5}
 {bone1Name=dyn_jacket_back_02_l, bone2Name=dyn_jacket_back_01_l, swing1LimitDegrees=15, swing2LimitDegrees=45, twistLimitDegrees=5}
 {
  bone1Name=spine_02
  bone2Name=spine_01
  swing1LimitDegrees=45
  swing1Motion=Free
  swing2LimitDegrees=45
  swing2Motion=Free
  twistLimitDegrees=45
  twistMotion=Free
 }
 {
  bone1Name=spine_01
  bone2Name=pelvis
  swing1LimitDegrees=45
  swing1Motion=Free
  swing2LimitDegrees=45
  swing2Motion=Free
  twistLimitDegrees=45
  twistMotion=Free
 }
 {
  bone1Name=thigh_r
  bone2Name=pelvis
  swing1LimitDegrees=45
  swing1Motion=Free
  swing2LimitDegrees=45
  swing2Motion=Free
  twistLimitDegrees=45
  twistMotion=Free
 }
 {
  bone1Name=thigh_l
  bone2Name=pelvis
  swing1LimitDegrees=45
  swing1Motion=Free
  swing2LimitDegrees=45
  swing2Motion=Free
  twistLimitDegrees=45
  twistMotion=Free
 }
 {bone1Name=dyn_hoodie_01_r, bone2Name=dyn_hoodie_01_r_MSTR, swing1LimitDegrees=5, swing2LimitDegrees=15, twistLimitDegrees=5}
 {
  bone1Name=spine_05
  bone2Name=spine_04
  swing1LimitDegrees=45
  swing1Motion=Free
  swing2LimitDegrees=45
  swing2Motion=Free
  twistLimitDegrees=45
  twistMotion=Free
 }
 {bone1Name=dyn_jacket_back_01_r, bone2Name=dyn_jacket_back_01_r_MSTR, swing1LimitDegrees=15, swing2LimitDegrees=45, twistLimitDegrees=5}
 {bone1Name=dyn_jacket_back_02_r, bone2Name=dyn_jacket_back_01_r, swing1LimitDegrees=15, swing2LimitDegrees=45, twistLimitDegrees=5}
 {bone1Name=dyn_jacket_front_01_r, bone2Name=dyn_jacket_front_01_r_MSTR, swing1LimitDegrees=15, swing2LimitDegrees=45, twistLimitDegrees=5}
 {bone1Name=dyn_jacket_front_02_r, bone2Name=dyn_jacket_front_01_r, swing1LimitDegrees=15, swing2LimitDegrees=45, twistLimitDegrees=5}
 {bone1Name=dyn_hoodie_zipper_01_r, bone2Name=dyn_jacket_front_01_r, swing1LimitDegrees=45, swing2LimitDegrees=45, twistLimitDegrees=45}
 {bone1Name=dyn_jacket_side_01_r, bone2Name=dyn_jacket_side_01_r_MSTR, swing1LimitDegrees=15, swing2LimitDegrees=35, twistLimitDegrees=5}
 {bone1Name=dyn_jacket_side_02_r, bone2Name=dyn_jacket_side_01_r, swing1LimitDegrees=15, swing2LimitDegrees=35, twistLimitDegrees=5}
 {bone1Name=dyn_jacket_side_01_l, bone2Name=dyn_jacket_side_01_l_MSTR, swing1LimitDegrees=15, swing2LimitDegrees=35, twistLimitDegrees=5}
 {bone1Name=dyn_jacket_side_02_l, bone2Name=dyn_jacket_side_01_l, swing1LimitDegrees=15, swing2LimitDegrees=35, twistLimitDegrees=5}
 {bone1Name=dyn_jacket_front_02_l, bone2Name=dyn_jacket_front_01_l, swing1LimitDegrees=15, swing2LimitDegrees=45, twistLimitDegrees=5}
 {bone1Name=dyn_jacket_front_01_l, bone2Name=dyn_jacket_front_01_l_MSTR, swing1LimitDegrees=15, swing2LimitDegrees=45, twistLimitDegrees=5}
]
