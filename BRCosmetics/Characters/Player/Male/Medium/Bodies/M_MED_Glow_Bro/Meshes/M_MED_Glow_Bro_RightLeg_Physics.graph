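# Physics asset for the right leg of the M_MED_Glow_Bro player body, covering a dynamic flap on that leg.
# Three box bodies form a constrained chain hanging from a tiny capsule on the flap's MSTR bone.
# Capsules on the pelvis and right calf twist bone supply leg and hip volumes.
# The MSTR and calf bodies link to the pelvis with free swing and twist.
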PhysicsAsset /BRCosmetics/Characters/Player/Male/Medium/Bodies/M_MED_Glow_Bro/Meshes/M_MED_Glow_Bro_RightLeg_Physics
bodies {
 calf_twist_02_r [
  {shapeType=Capsule, length=32.9497, radius=7.1167, rotation=<rot -90 -1.8984 1.8984>}
  {shapeType=Capsule, length=13.5, radius=7.25, center=<-2.7987 3.2369 -0.0004>, rotation=<rot -90 174.9489 -174.9488>}
 ]
 dyn_legs_flap_01_r [{shapeType=Box, extentX=6.0561, extentY=3.4384, extentZ=1.7393, center=<-2.8831 0.0001 0>, rotation=<rot 0 0 180>}]
 dyn_legs_flap_01_r_MSTR [{shapeType=Capsule, length=0.01, radius=0.01}]
 dyn_legs_flap_02_r [{shapeType=Box, extentX=6.0561, extentY=3.4384, extentZ=1.7393, center=<-2.8831 0.0001 0>, rotation=<rot 0 0 180>}]
 dyn_legs_flap_03_r [{shapeType=Box, extentX=6.0561, extentY=3.4384, extentZ=1.7393, center=<-2.8831 0.0001 0>, rotation=<rot 0 0 180>}]
 pelvis [{shapeType=Capsule, length=10.8441, radius=10.2509}]
}
constraints [
 {
  bone1Name=dyn_legs_flap_01_r_MSTR
  bone2Name=pelvis
  swing1LimitDegrees=45
  swing1Motion=Free
  swing2LimitDegrees=45
  swing2Motion=Free
  twistLimitDegrees=45
  twistMotion=Free
 }
 {
  bone1Name=calf_twist_02_r
  bone2Name=pelvis
  swing1LimitDegrees=45
  swing1Motion=Free
  swing2LimitDegrees=45
  swing2Motion=Free
  twistLimitDegrees=45
  twistMotion=Free
 }
 {bone1Name=dyn_legs_flap_01_r, bone2Name=dyn_legs_flap_01_r_MSTR, swing1LimitDegrees=45, swing2LimitDegrees=45, twistLimitDegrees=45}
 {bone1Name=dyn_legs_flap_02_r, bone2Name=dyn_legs_flap_01_r, swing1LimitDegrees=45, swing2LimitDegrees=45, twistLimitDegrees=45}
 {bone1Name=dyn_legs_flap_03_r, bone2Name=dyn_legs_flap_02_r, swing1LimitDegrees=45, swing2LimitDegrees=45, twistLimitDegrees=45}
]
