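# Post-process style AnimBlueprint for the Glow Bro character body, adding secondary motion to hoodie, shirt, jacket and flaps.
# The copied parent pose gets bone fixes driven by crouch, skydive and thigh angle, then RigidBody and AnimDynamics, before caching.
# Low LOD or disabled rigid bodies selects posed, constrained cords; otherwise tassel RigidBody crossfades to cord AnimDynamics during emotes.
# Arm-tassel and leg RigidBody nodes are unconnected.
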AnimBlueprint /BRCosmetics/Characters/Player/Male/Medium/Bodies/M_MED_Glow_Bro/Meshes/M_MED_Glow_Bro_AnimBP
target_skeleton=&/M_MED_Glow_Bro_Skeleton
graphs [
 AnimGraph {nodes [
   n1 = Break Fort Anim Input Skydiving External Force(FortAnimInput_SkydivingExternalForce=Get SkydivingExternalForceInput()) -> (
    bUseSkydivingVectorForce=~
    SkydivingVectorForce=~
    HeadToPelvisDirection=~
    FloatingMultiplier=~
    FloatingAdditive=~
    DivingMultiplier=~
    DivingAdditive=~
    ParachutingMultiplier=~
    ParachutingAdditive=~
   )
   RigidBody(Alpha=0.35, ExternalForce=@n1.SkydivingVectorForce) {
    bClampLinearTranslationLimitToRefPose=true
    bOverrideWorldGravity=true
    overridePhysicsAsset=&/M_MED_Glow_Bro_ArmTassles_Physics
    simulationSpace=BaseBoneSpace
    alphaBoolBlend {blendOption=Linear}
    baseBoneRef {boneName=pelvis}
    componentAppliedLinearAccClamp=<3000 3000 3000>
    componentLinearAccScale=<0.7 0.7 0.7>
    componentLinearVelScale=<0.7 0.7 0.7>
    overrideWorldGravity {z=-980}
   }
   RigidBody(ExternalForce=@n1.SkydivingVectorForce) {
    bClampLinearTranslationLimitToRefPose=true
    bOverrideWorldGravity=true
    overridePhysicsAsset=&/M_MED_Glow_Bro_RightLeg_Physics
    simulationSpace=BaseBoneSpace
    alphaBoolBlend {blendOption=Linear}
    baseBoneRef {boneName=thigh_twist_01_r}
    componentAppliedLinearAccClamp=<3000 3000 3000>
    componentLinearAccScale=<0.5 0.5 0.75>
    componentLinearVelScale=<0.5 0.5 0.75>
    overrideWorldGravity {z=-980}
   }
   RigidBody(ExternalForce=@n1.SkydivingVectorForce) {
    bClampLinearTranslationLimitToRefPose=true
    bOverrideWorldGravity=true
    overridePhysicsAsset=&/M_MED_Glow_Bro_leftLeg_Physics
    simulationSpace=BaseBoneSpace
    alphaBoolBlend {blendOption=Linear}
    baseBoneRef {boneName=thigh_twist_01_l}
    componentAppliedLinearAccClamp=<3000 3000 3000>
    componentLinearAccScale=<0.5 0.5 0.75>
    componentLinearVelScale=<0.5 0.5 0.75>
    overrideWorldGravity {z=-980}
   }
   n2 = "Constraint - dyn_hoodie_cord_03_r (P)"(ComponentPose=Local To Component(LocalPose=Use cached pose 'cachedPose'()), Alpha=1, ConstraintWeights_0=1) {
    alphaBoolBlend {blendOption=Linear}
    boneToModify {boneName=dyn_hoodie_cord_03_r}
    constraintSetup [{targetBone {boneName=spine_03}}]
    constraintWeights [1]
   }
   n3 = "Constraint - dyn_hoodie_cord_04_r (P)"(ComponentPose=@n2, Alpha=1, ConstraintWeights_0=1) {
    alphaBoolBlend {blendOption=Linear}
    boneToModify {boneName=dyn_hoodie_cord_04_r}
    constraintSetup [{targetBone {boneName=spine_03}}]
    constraintWeights [1]
   }
   n4 = "Transform (Modify) Bone - Bone: dyn_hoodie_cord_02_r"(ComponentPose=@n3, Alpha=1, Translation="0,0,0.5", Rotation="0,5,0", Scale="1,1,1") {
    rotationMode=BMM_Additive
    rotationSpace=BCS_BoneSpace
    translationMode=BMM_Additive
    translationSpace=BCS_BoneSpace
    alphaBoolBlend {blendOption=Linear}
    boneToModify {boneName=dyn_hoodie_cord_02_r}
    rotation {yaw=5}
    translation {z=0.5}
   }
   n5 = "Transform (Modify) Bone - Bone: dyn_hoodie_cord_03_r"(ComponentPose=@n4, Alpha=1, Translation="0,0,0.5", Rotation="0,20,0", Scale="1,1,1") {
    rotationMode=BMM_Additive
    rotationSpace=BCS_BoneSpace
    translationMode=BMM_Additive
    translationSpace=BCS_BoneSpace
    alphaBoolBlend {blendOption=Linear}
    boneToModify {boneName=dyn_hoodie_cord_03_r}
    rotation {yaw=20}
    translation {z=0.5}
   }
   n6 = "Transform (Modify) Bone - Bone: dyn_hoodie_cord_04_r"(ComponentPose=@n5, Alpha=1, Translation="0,0,0.5", Rotation="0,10,0", Scale="1,1,1") {
    rotationMode=BMM_Additive
    rotationSpace=BCS_BoneSpace
    translationMode=BMM_Additive
    translationSpace=BCS_BoneSpace
    alphaBoolBlend {blendOption=Linear}
    boneToModify {boneName=dyn_hoodie_cord_04_r}
    rotation {yaw=10}
    translation {z=0.5}
   }
   n7 = "Constraint - dyn_hoodie_cord_03_l (P)"(ComponentPose=@n6, Alpha=1, ConstraintWeights_0=1) {
    alphaBoolBlend {blendOption=Linear}
    boneToModify {boneName=dyn_hoodie_cord_03_l}
    constraintSetup [{targetBone {boneName=spine_03}}]
    constraintWeights [1]
   }
   n8 = "Constraint - dyn_hoodie_cord_04_l (P)"(ComponentPose=@n7, Alpha=1, ConstraintWeights_0=1) {
    alphaBoolBlend {blendOption=Linear}
    boneToModify {boneName=dyn_hoodie_cord_04_l}
    constraintSetup [{targetBone {boneName=spine_03}}]
    constraintWeights [1]
   }
   n9 = "Transform (Modify) Bone - Bone: dyn_hoodie_cord_02_l"(ComponentPose=@n8, Alpha=1, Translation="0,0,0.5", Rotation="0,5,0", Scale="1,1,1") {
    rotationMode=BMM_Additive
    rotationSpace=BCS_BoneSpace
    translationMode=BMM_Additive
    translationSpace=BCS_BoneSpace
    alphaBoolBlend {blendOption=Linear}
    boneToModify {boneName=dyn_hoodie_cord_02_l}
    rotation {yaw=5}
    translation {z=0.5}
   }
   n10 = "Transform (Modify) Bone - Bone: dyn_hoodie_cord_03_l"(ComponentPose=@n9, Alpha=1, Translation="0,0,0.5", Rotation="0,20,0", Scale="1,1,1") {
    rotationMode=BMM_Additive
    rotationSpace=BCS_BoneSpace
    translationMode=BMM_Additive
    translationSpace=BCS_BoneSpace
    alphaBoolBlend {blendOption=Linear}
    boneToModify {boneName=dyn_hoodie_cord_03_l}
    rotation {yaw=20}
    translation {z=0.5}
   }
   n11 = "Transform (Modify) Bone - Bone: dyn_hoodie_cord_04_l"(ComponentPose=@n10, Alpha=1, Translation="0,0,0.5", Rotation="0,10,0", Scale="1,1,1") {
    rotationMode=BMM_Additive
    rotationSpace=BCS_BoneSpace
    translationMode=BMM_Additive
    translationSpace=BCS_BoneSpace
    alphaBoolBlend {blendOption=Linear}
    boneToModify {boneName=dyn_hoodie_cord_04_l}
    rotation {yaw=10}
    translation {z=0.5}
   }
   n12 = RigidBody(
    ComponentPose=Local To Component(LocalPose=Use cached pose 'cachedPose'())
    Alpha=Get isEmotingAlphaInv() -> (isEmotingAlphaInv:double)
    ExternalForce=@n1.SkydivingVectorForce
   ) {
    bClampLinearTranslationLimitToRefPose=true
    bOverrideWorldGravity=true
    overridePhysicsAsset=&/M_MED_Glow_Bro_CoatTassles_Physics
    simulationSpace=BaseBoneSpace
    alphaBoolBlend {blendOption=Linear}
    baseBoneRef {boneName=spine_01}
    componentAppliedLinearAccClamp=<3000 3000 3000>
    componentLinearAccScale=<0.3 0.3 0.3>
    componentLinearVelScale=<0.3 0.3 0.7>
    overrideWorldGravity {z=-980}
   }
   n13 = Get isEmotingAlpha() -> (isEmotingAlpha:double)
   n14 = "Transform (Modify) Bone - Bone: dyn_hoodie_cord_03_l"(ComponentPose=@n12, Alpha=@n13, Translation="0,0,0", Rotation="0,0,30", Scale="1,1,1") {rotationMode=BMM_Additive, alphaBoolBlend {blendOption=Linear}, boneToModify {boneName=dyn_hoodie_cord_03_l}, rotation {roll=30}}
   n15 = Anim Dynamics - Bone: dyn_hoodie_cord_03_l(ComponentPose=@n14, Alpha=@n13, ExternalForce=@n1.SkydivingVectorForce) {
    angularDampingOverride=1
    angularSpringConstant=100
    bAngularSpring=true
    bGravityOverrideInSimSpace=true
    bOverrideAngularDamping=true
    bUseGravityOverride=true
    bUseSphericalLimits=true
    gravityScale=0.25
    alphaBoolBlend {blendOption=Linear}
    boundBone {boneName=dyn_hoodie_cord_03_l}
    chainEnd {boneName=dyn_hoodie_cord_05_l}
    componentAppliedLinearAccClamp=<3000 3000 3000>
    componentLinearAccScale=<0.7 0.7 0.7>
    componentLinearVelScale=<0.7 0.7 0.7>
    gravityOverride {z=-980}
    physicsBodyDefinitions [{
      collisionType=CustomSphere
      sphereCollisionRadius=1
      boundBone {boneName=dyn_hoodie_cord_03_l}
      boxExtents=<10 10 10>
      constraintSetup {angularLimitsMax=<5 15 20>, angularLimitsMin=<-5 -15 -45>, angularTarget=<0 0 0>}
      localJointOffset=<6 0 0>
     }]
    relativeSpaceBone {boneName=spine_05}
    sphericalLimits [{limitRadius=25, drivingBone {boneName=spine_05}, sphereLocalOffset=<-7 -10 0>}]
   }
   n16 = Anim Dynamics - Bone: dyn_hoodie_cord_04_l(ComponentPose=@n15, Alpha=@n13, ExternalForce=@n1.SkydivingVectorForce) {
    angularDampingOverride=1
    angularSpringConstant=100
    bAngularSpring=true
    bGravityOverrideInSimSpace=true
    bOverrideAngularDamping=true
    bUseGravityOverride=true
    bUseSphericalLimits=true
    gravityScale=0.25
    alphaBoolBlend {blendOption=Linear}
    boundBone {boneName=dyn_hoodie_cord_04_l}
    chainEnd {boneName=dyn_hoodie_cord_05_l}
    componentAppliedLinearAccClamp=<3000 3000 3000>
    componentLinearAccScale=<0.7 0.7 0.7>
    componentLinearVelScale=<0.7 0.7 0.7>
    gravityOverride {z=-980}
    physicsBodyDefinitions [{
      collisionType=CustomSphere
      sphereCollisionRadius=1
      boundBone {boneName=dyn_hoodie_cord_04_l}
      boxExtents=<10 10 10>
      constraintSetup {angularLimitsMax=<5 15 20>, angularLimitsMin=<-5 -15 -45>, angularTarget=<0 0 0>}
      localJointOffset=<6 0 0>
     }]
    relativeSpaceBone {boneName=spine_05}
    sphericalLimits [
     {limitRadius=25, drivingBone {boneName=spine_05}, sphereLocalOffset=<-7 -10 0>}
     {limitRadius=25, drivingBone {boneName=spine_02}, sphereLocalOffset=<-7 -8 0>}
    ]
   }
   n17 = Anim Dynamics - Bone: dyn_hoodie_cord_05_l(ComponentPose=@n16, Alpha=@n13, ExternalForce=@n1.SkydivingVectorForce) {
    angularDampingOverride=1
    angularSpringConstant=100
    bAngularSpring=true
    bGravityOverrideInSimSpace=true
    bOverrideAngularDamping=true
    bUseGravityOverride=true
    bUseSphericalLimits=true
    gravityScale=0.25
    alphaBoolBlend {blendOption=Linear}
    boundBone {boneName=dyn_hoodie_cord_05_l}
    chainEnd {boneName=dyn_hoodie_cord_05_l}
    componentAppliedLinearAccClamp=<3000 3000 3000>
    componentLinearAccScale=<0.7 0.7 0.7>
    componentLinearVelScale=<0.7 0.7 0.7>
    gravityOverride {z=-980}
    physicsBodyDefinitions [{
      collisionType=CustomSphere
      sphereCollisionRadius=1
      boundBone {boneName=dyn_hoodie_cord_05_l}
      boxExtents=<10 10 10>
      constraintSetup {angularLimitsMax=<5 15 20>, angularLimitsMin=<-5 -15 -45>, angularTarget=<0 0 0>}
      localJointOffset=<6 0 0>
     }]
    relativeSpaceBone {boneName=spine_05}
    sphericalLimits [
     {limitRadius=25, drivingBone {boneName=spine_05}, sphereLocalOffset=<-7 -10 0>}
     {limitRadius=25, drivingBone {boneName=spine_02}, sphereLocalOffset=<0.7 -8 0>}
    ]
   }
   n18 = "Transform (Modify) Bone - Bone: dyn_hoodie_cord_03_r"(ComponentPose=@n17, Alpha=@n13, Translation="0,0,0", Rotation="0,0,-30", Scale="1,1,1") {rotationMode=BMM_Additive, alphaBoolBlend {blendOption=Linear}, boneToModify {boneName=dyn_hoodie_cord_03_r}, rotation {roll=-30}}
   n19 = Anim Dynamics - Bone: dyn_hoodie_cord_03_r(ComponentPose=@n18, Alpha=@n13, ExternalForce=@n1.SkydivingVectorForce) {
    angularDampingOverride=1
    angularSpringConstant=100
    bAngularSpring=true
    bGravityOverrideInSimSpace=true
    bOverrideAngularDamping=true
    bUseGravityOverride=true
    bUseSphericalLimits=true
    gravityScale=0.25
    alphaBoolBlend {blendOption=Linear}
    boundBone {boneName=dyn_hoodie_cord_03_r}
    chainEnd {boneName=dyn_hoodie_cord_05_l}
    componentAppliedLinearAccClamp=<3000 3000 3000>
    componentLinearAccScale=<0.7 0.7 0.7>
    componentLinearVelScale=<0.7 0.7 0.7>
    gravityOverride {z=-980}
    physicsBodyDefinitions [{
      collisionType=CustomSphere
      sphereCollisionRadius=1
      boundBone {boneName=dyn_hoodie_cord_03_r}
      boxExtents=<10 10 10>
      constraintSetup {angularLimitsMax=<5 15 45>, angularLimitsMin=<-5 -15 0>, angularTarget=<0 0 0>}
      localJointOffset=<-6 0 0>
     }]
    relativeSpaceBone {boneName=spine_05}
    sphericalLimits [{limitRadius=30, drivingBone {boneName=spine_05}, sphereLocalOffset=<-7 -15 0>}]
   }
   n20 = Anim Dynamics - Bone: dyn_hoodie_cord_04_r(ComponentPose=@n19, Alpha=@n13, ExternalForce=@n1.SkydivingVectorForce) {
    angularDampingOverride=1
    angularSpringConstant=100
    bAngularSpring=true
    bGravityOverrideInSimSpace=true
    bOverrideAngularDamping=true
    bUseGravityOverride=true
    bUseSphericalLimits=true
    gravityScale=0.25
    alphaBoolBlend {blendOption=Linear}
    boundBone {boneName=dyn_hoodie_cord_04_r}
    chainEnd {boneName=dyn_hoodie_cord_05_l}
    componentAppliedLinearAccClamp=<3000 3000 3000>
    componentLinearAccScale=<0.7 0.7 0.7>
    componentLinearVelScale=<0.7 0.7 0.7>
    gravityOverride {z=-980}
    physicsBodyDefinitions [{
      collisionType=CustomSphere
      sphereCollisionRadius=1
      boundBone {boneName=dyn_hoodie_cord_04_r}
      boxExtents=<10 10 10>
      constraintSetup {angularLimitsMax=<5 15 45>, angularLimitsMin=<-5 -15 0>, angularTarget=<0 0 0>}
      localJointOffset=<-6 0 0>
     }]
    relativeSpaceBone {boneName=spine_05}
    sphericalLimits [
     {limitRadius=25, drivingBone {boneName=spine_05}, sphereLocalOffset=<-7 -10 0>}
     {limitRadius=25, drivingBone {boneName=spine_02}, sphereLocalOffset=<-7 -8 0>}
    ]
   }
   n21 = Anim Dynamics - Bone: dyn_hoodie_cord_05_r(ComponentPose=@n20, Alpha=@n13, ExternalForce=@n1.SkydivingVectorForce) {
    angularDampingOverride=1
    angularSpringConstant=100
    bAngularSpring=true
    bGravityOverrideInSimSpace=true
    bOverrideAngularDamping=true
    bUseGravityOverride=true
    bUseSphericalLimits=true
    gravityScale=0.25
    alphaBoolBlend {blendOption=Linear}
    boundBone {boneName=dyn_hoodie_cord_05_r}
    chainEnd {boneName=dyn_hoodie_cord_05_l}
    componentAppliedLinearAccClamp=<3000 3000 3000>
    componentLinearAccScale=<0.7 0.7 0.7>
    componentLinearVelScale=<0.7 0.7 0.7>
    gravityOverride {z=-980}
    physicsBodyDefinitions [{
      collisionType=CustomSphere
      sphereCollisionRadius=1
      boundBone {boneName=dyn_hoodie_cord_05_r}
      boxExtents=<10 10 10>
      constraintSetup {angularLimitsMax=<5 15 45>, angularLimitsMin=<-5 -15 0>, angularTarget=<0 0 0>}
      localJointOffset=<-6 0 0>
     }]
    relativeSpaceBone {boneName=spine_05}
    sphericalLimits [
     {limitRadius=25, drivingBone {boneName=spine_05}, sphereLocalOffset=<-7 -10 0>}
     {limitRadius=25, drivingBone {boneName=spine_02}, sphereLocalOffset=<-7 -8 0>}
    ]
   }
   n22 = Blend Poses by bool(
    BlendPose_0=Component To Local(ComponentPose=@n11)
    BlendPose_1=Component To Local(ComponentPose=@n21)
    bActiveValue=Get isMobileLOD() -> (isMobileLOD:bool)
   ) {blendType=Linear}
   Output Pose(Result=Component To Local(ComponentPose=Local To Component(LocalPose=@n22))) {layerGroup=None}
   n23 = "Transform (Modify) Bone - Bone: shirt_pelvis_01_DSP"(
    ComponentPose=Local To Component(LocalPose=Copy Pose From Mesh() {bUseAttachedParent=true})
    Alpha=1
    Translation="0,-5,0"
    Rotation="0,0,0"
    Scale="1,1,1"
   ) {translationMode=BMM_Additive, alphaBoolBlend {blendOption=Linear}, boneToModify {boneName=shirt_pelvis_01_DSP}, translation {y=-5}}
   n24 = "Transform (Modify) Bone - Bone: dyn_pelvis_01_DSP"(ComponentPose=@n23, Alpha=1, Translation="0,0,0", Rotation="0,0,0", Scale="-0.7,-0.7,-0.7") {
    rotationMode=BMM_Additive
    scaleMode=BMM_Additive
    translationMode=BMM_Additive
    alphaBoolBlend {blendOption=Linear}
    boneToModify {boneName=dyn_pelvis_01_DSP}
    scale=<-0.7 -0.7 -0.7>
   }
   n25 = "Transform (Modify) Bone - Bone: thigh_twist_pants_01_l_DSP"(ComponentPose=@n24, Alpha=1, Translation="0,0,0", Rotation="0,0,-4", Scale="1,1,1") {
    rotationMode=BMM_Additive
    scaleMode=BMM_Additive
    translationMode=BMM_Additive
    alphaBoolBlend {blendOption=Linear}
    boneToModify {boneName=thigh_twist_pants_01_l_DSP}
    rotation {roll=-4}
   }
   n26 = "Constraint - dyn_shirt_front_pelvis_01 (R)"(ComponentPose=@n25, Alpha=1, ConstraintWeights_0=1) {
    alphaBoolBlend {blendOption=Linear}
    boneToModify {boneName=dyn_shirt_front_pelvis_01}
    constraintSetup [{transformType=Rotation, targetBone {boneName=thigh_twist_01_l}}]
    constraintWeights [1]
   }
   n27 = "Constraint - dyn_shirt_front_01_l_MSTR (P)"(ComponentPose=@n26, Alpha=1, ConstraintWeights_0=1) {
    alphaBoolBlend {blendOption=Linear}
    boneToModify {boneName=dyn_shirt_front_01_l_MSTR}
    constraintSetup [{targetBone {boneName=thigh_twist_01_l}}]
    constraintWeights [1]
   }
   n28 = Get isCrouchingAlpha() -> (isCrouchingAlpha:double)
   n29 = "Transform (Modify) Bone - Bone: thigh_twist_pants_01_l_DSP"(ComponentPose=@n27, Alpha=@n28, Translation="0,0,0", Rotation="0,0,10", Scale="1,1,1") {
    rotationMode=BMM_Additive
    scaleMode=BMM_Additive
    translationMode=BMM_Additive
    alphaBoolBlend {blendOption=Linear}
    boneToModify {boneName=thigh_twist_pants_01_l_DSP}
    rotation {roll=10}
   }
   n30 = "Transform (Modify) Bone - Bone: dyn_shirt_front_pelvis_01"(ComponentPose=@n29, Alpha=@n28, Translation="-2.5,-4,0", Rotation="0,0,25", Scale="1,1,1") {
    rotationMode=BMM_Additive
    scaleMode=BMM_Additive
    translationMode=BMM_Additive
    alphaBoolBlend {blendOption=Linear}
    boneToModify {boneName=dyn_shirt_front_pelvis_01}
    rotation {roll=25}
    translation=<-2.5 -4>
   }
   n31 = "Transform (Modify) Bone - Bone: thigh_twist_pants_01_r_DSP"(ComponentPose=@n30, Alpha=1, Translation="0,0,0", Rotation="-8,0,-4", Scale="1,1,1") {
    rotationMode=BMM_Additive
    scaleMode=BMM_Additive
    translationMode=BMM_Additive
    alphaBoolBlend {blendOption=Linear}
    boneToModify {boneName=thigh_twist_pants_01_r_DSP}
    rotation {pitch=-8, roll=-4}
   }
   n32 = "Transform (Modify) Bone - Bone: dyn_shirt_front_01_r"(ComponentPose=@n31, Alpha=1, Translation="0,0,0", Rotation="0,0,-12", Scale="1,1,1") {
    rotationMode=BMM_Additive
    scaleMode=BMM_Additive
    translationMode=BMM_Additive
    alphaBoolBlend {blendOption=Linear}
    boneToModify {boneName=dyn_shirt_front_01_r}
    rotation {roll=-12}
   }
   n33 = "Transform (Modify) Bone - Bone: dyn_shirt_front_01_l"(ComponentPose=@n32, Alpha=1, Translation="0,0,0", Rotation="0,0,-12", Scale="1,1,1") {
    rotationMode=BMM_Additive
    scaleMode=BMM_Additive
    translationMode=BMM_Additive
    alphaBoolBlend {blendOption=Linear}
    boneToModify {boneName=dyn_shirt_front_01_l}
    rotation {roll=-12}
   }
   n34 = "Transform (Modify) Bone - Bone: dyn_shirt_front_01_l"(ComponentPose=@n33, Alpha=Get isCrouchingAlpha(), Translation="-2,0,0", Rotation="0,0,15", Scale="1,1,1") {
    rotationMode=BMM_Additive
    scaleMode=BMM_Additive
    translationMode=BMM_Additive
    alphaBoolBlend {blendOption=Linear}
    boneToModify {boneName=dyn_shirt_front_01_l}
    rotation {roll=15}
    translation {x=-2}
   }
   n35 = Get isFloatingAlpha() -> (isFloatingAlpha:double)
   n36 = "Transform (Modify) Bone - Bone: dyn_jacket_back_01_l_MSTR"(ComponentPose=@n34, Alpha=@n35, Translation="0,0,0", Rotation="0,0,45", Scale="1,1,1") {
    rotationMode=BMM_Additive
    scaleMode=BMM_Additive
    translationMode=BMM_Additive
    alphaBoolBlend {blendOption=Linear}
    boneToModify {boneName=dyn_jacket_back_01_l_MSTR}
    rotation {roll=45}
   }
   n37 = "Transform (Modify) Bone - Bone: dyn_jacket_back_01_r_MSTR"(ComponentPose=@n36, Alpha=@n35, Translation="0,0,0", Rotation="0,0,45", Scale="1,1,1") {
    rotationMode=BMM_Additive
    scaleMode=BMM_Additive
    translationMode=BMM_Additive
    alphaBoolBlend {blendOption=Linear}
    boneToModify {boneName=dyn_jacket_back_01_r_MSTR}
    rotation {roll=45}
   }
   n38 = "Transform (Modify) Bone - Bone: pelvis_back_DSP"(ComponentPose=@n37, Alpha=@n35, Translation="0,0,0", Rotation="0,0,-10", Scale="0.9,0.9,1") {
    rotationMode=BMM_Additive
    scaleMode=BMM_Additive
    translationMode=BMM_Additive
    alphaBoolBlend {blendOption=Linear}
    boneToModify {boneName=pelvis_back_DSP}
    rotation {roll=-10}
    scale=<0.9 0.9>
   }
   n39 = RigidBody(ComponentPose=@n38, ExternalForce=@n1.SkydivingVectorForce) {
    bClampLinearTranslationLimitToRefPose=true
    bOverrideWorldGravity=true
    overridePhysicsAsset=&/M_MED_Glow_Bro_Physics
    simulationSpace=BaseBoneSpace
    alphaBoolBlend {blendOption=Linear}
    baseBoneRef {boneName=spine_01}
    componentAppliedLinearAccClamp=<3000 3000 3000>
    componentLinearAccScale=<0.3 0.3 0.3>
    componentLinearVelScale=<0.3 0.3 0.7>
    overrideWorldGravity {z=-980}
   }
   n40 = Anim Dynamics - Bone: dyn_hoodie_armFlap_01_l(ComponentPose=@n39, Alpha=1, ExternalForce="0,0,0") {
    angularDampingOverride=0.999
    angularSpringConstant=800
    bAngularSpring=true
    bGravityOverrideInSimSpace=true
    bOverrideAngularDamping=true
    gravityScale=0.2
    simulationSpace=RootRelative
    alphaBoolBlend {blendOption=Linear}
    boundBone {boneName=dyn_hoodie_armFlap_01_l}
    componentAppliedLinearAccClamp=<10000 10000 10000>
    componentLinearAccScale=<0.1 0.1 0.1>
    componentLinearVelScale=<0.1 0.1 0.1>
    physicsBodyDefinitions [{
      sphereCollisionRadius=0
      boundBone {boneName=dyn_hoodie_armFlap_01_l}
      boxExtents=<8 8 5>
      constraintSetup {angularTargetAxis=AxisY, twistAxis=AxisY, angularLimitsMax=<10 20 10>, angularLimitsMin=<-10 -20 -10>, angularTarget=<0 0 0>}
      localJointOffset=<-1 -2 4>
     }]
    relativeSpaceBone {boneName=pelvis}
   }
   n41 = Anim Dynamics - Bone: dyn_hoodie_armFlap_01_r(ComponentPose=@n40, Alpha=1, ExternalForce="0,0,0") {
    angularDampingOverride=0.999
    angularSpringConstant=800
    bAngularSpring=true
    bGravityOverrideInSimSpace=true
    bOverrideAngularDamping=true
    gravityScale=0.2
    simulationSpace=RootRelative
    alphaBoolBlend {blendOption=Linear}
    boundBone {boneName=dyn_hoodie_armFlap_01_r}
    componentAppliedLinearAccClamp=<10000 10000 10000>
    componentLinearAccScale=<0.1 0.1 0.1>
    componentLinearVelScale=<0.1 0.1 0.1>
    physicsBodyDefinitions [{
      sphereCollisionRadius=0
      boundBone {boneName=dyn_hoodie_armFlap_01_r}
      boxExtents=<8 8 5>
      constraintSetup {angularTargetAxis=AxisY, twistAxis=AxisY, angularLimitsMax=<10 20 10>, angularLimitsMin=<-10 -20 -10>, angularTarget=<0 0 0>}
      localJointOffset=<-1 -2 4>
     }]
    relativeSpaceBone {boneName=pelvis}
   }
   n42 = Get thigh_shape_fn_l() -> (thigh_shape_fn_l:double)
   n43 = "Transform (Modify) Bone - Bone: dyn_legs_flap_01_l"(ComponentPose=@n41, Alpha=@n42, Translation="0,0,0", Rotation="0,0,50", Scale="1,1,1") {
    rotationMode=BMM_Additive
    scaleMode=BMM_Additive
    translationMode=BMM_Additive
    alphaBoolBlend {blendOption=Linear}
    boneToModify {boneName=dyn_legs_flap_01_l}
    rotation {roll=50}
   }
   n44 = "Transform (Modify) Bone - Bone: dyn_legs_flap_02_l"(ComponentPose=@n43, Alpha=@n42, Translation="0,0,0", Rotation="0,0,30", Scale="1,1,1") {
    rotationMode=BMM_Additive
    scaleMode=BMM_Additive
    translationMode=BMM_Additive
    alphaBoolBlend {blendOption=Linear}
    boneToModify {boneName=dyn_legs_flap_02_l}
    rotation {roll=30}
   }
   n45 = Anim Dynamics - Bone: dyn_legs_flap_01_l(ComponentPose=@n44, Alpha=1, ExternalForce="0,0,0") {
    angularDampingOverride=0.999
    angularSpringConstant=800
    bAngularSpring=true
    bGravityOverrideInSimSpace=true
    bOverrideAngularDamping=true
    gravityScale=0.2
    simulationSpace=RootRelative
    alphaBoolBlend {blendOption=Linear}
    boundBone {boneName=dyn_legs_flap_01_l}
    componentAppliedLinearAccClamp=<10000 10000 10000>
    componentLinearAccScale=<0.1 0.1 0.1>
    componentLinearVelScale=<0.1 0.1 0.1>
    gravityOverride {z=-980}
    physicsBodyDefinitions [{
      sphereCollisionRadius=0
      boundBone {boneName=dyn_legs_flap_01_l}
      boxExtents=<5 14 14>
      constraintSetup {twistAxis=AxisZ, angularLimitsMax=<5 5 10>, angularLimitsMin=<-5 -30 -10>, angularTarget=<0 0 0>}
      localJointOffset=<5 -1 0>
     }]
    relativeSpaceBone {boneName=dyn_legs_flap_01_l_MSTR}
   }
   n46 = Anim Dynamics - Bone: dyn_legs_flap_02_l(ComponentPose=@n45, Alpha=1, ExternalForce="0,0,0") {
    angularDampingOverride=0.999
    angularSpringConstant=800
    bAngularSpring=true
    bGravityOverrideInSimSpace=true
    bOverrideAngularDamping=true
    gravityScale=0.2
    simulationSpace=RootRelative
    alphaBoolBlend {blendOption=Linear}
    boundBone {boneName=dyn_legs_flap_02_l}
    componentAppliedLinearAccClamp=<10000 10000 10000>
    componentLinearAccScale=<0.1 0.1 0.1>
    componentLinearVelScale=<0.1 0.1 0.1>
    gravityOverride {z=-980}
    physicsBodyDefinitions [{
      sphereCollisionRadius=0
      boundBone {boneName=dyn_legs_flap_02_l}
      boxExtents=<5 14 14>
      constraintSetup {angularTargetAxis=AxisY, twistAxis=AxisZ, angularLimitsMax=<5 5 10>, angularLimitsMin=<-5 -30 -10>, angularTarget=<0 0 0>}
      localJointOffset=<5 -1 0>
     }]
    relativeSpaceBone {boneName=dyn_legs_flap_01_l_MSTR}
   }
   n47 = Anim Dynamics - Bone: dyn_legs_flap_03_l(ComponentPose=@n46, Alpha=1, ExternalForce="0,0,0") {
    angularDampingOverride=0.999
    angularSpringConstant=800
    bAngularSpring=true
    bGravityOverrideInSimSpace=true
    bOverrideAngularDamping=true
    gravityScale=0.2
    simulationSpace=RootRelative
    alphaBoolBlend {blendOption=Linear}
    boundBone {boneName=dyn_legs_flap_03_l}
    componentAppliedLinearAccClamp=<10000 10000 10000>
    componentLinearAccScale=<0.1 0.1 0.1>
    componentLinearVelScale=<0.1 0.1 0.1>
    gravityOverride {z=-980}
    physicsBodyDefinitions [{
      sphereCollisionRadius=0
      boundBone {boneName=dyn_legs_flap_03_l}
      boxExtents=<5 14 14>
      constraintSetup {angularTargetAxis=AxisY, twistAxis=AxisZ, angularLimitsMax=<5 5 10>, angularLimitsMin=<-5 -30 -10>, angularTarget=<0 0 0>}
      localJointOffset=<5 -1 0>
     }]
    relativeSpaceBone {boneName=dyn_legs_flap_01_l_MSTR}
   }
   n48 = Get thigh_shape_fn_r() -> (thigh_shape_fn_r:double)
   n49 = "Transform (Modify) Bone - Bone: dyn_legs_flap_01_r"(ComponentPose=@n47, Alpha=@n48, Translation="0,0,0", Rotation="0,0,50", Scale="1,1,1") {
    rotationMode=BMM_Additive
    scaleMode=BMM_Additive
    translationMode=BMM_Additive
    alphaBoolBlend {blendOption=Linear}
    boneToModify {boneName=dyn_legs_flap_01_r}
    rotation {roll=50}
   }
   n50 = "Transform (Modify) Bone - Bone: dyn_legs_flap_02_r"(ComponentPose=@n49, Alpha=@n48, Translation="0,0,0", Rotation="0,0,30", Scale="1,1,1") {
    rotationMode=BMM_Additive
    scaleMode=BMM_Additive
    translationMode=BMM_Additive
    alphaBoolBlend {blendOption=Linear}
    boneToModify {boneName=dyn_legs_flap_02_r}
    rotation {roll=30}
   }
   n51 = Anim Dynamics - Bone: dyn_legs_flap_01_r(ComponentPose=@n50, Alpha=1, ExternalForce="0,0,0") {
    angularDampingOverride=0.999
    angularSpringConstant=800
    bAngularSpring=true
    bGravityOverrideInSimSpace=true
    bOverrideAngularDamping=true
    gravityScale=0.2
    simulationSpace=RootRelative
    alphaBoolBlend {blendOption=Linear}
    boundBone {boneName=dyn_legs_flap_01_r}
    componentAppliedLinearAccClamp=<10000 10000 10000>
    componentLinearAccScale=<0.1 0.1 0.1>
    componentLinearVelScale=<0.1 0.1 0.1>
    gravityOverride {z=-980}
    physicsBodyDefinitions [{
      sphereCollisionRadius=0
      boundBone {boneName=dyn_legs_flap_01_r}
      boxExtents=<5 14 14>
      constraintSetup {twistAxis=AxisZ, angularLimitsMax=<5 5 10>, angularLimitsMin=<-5 -30 -10>, angularTarget=<0 0 0>}
      localJointOffset=<-5 -1 0>
     }]
    relativeSpaceBone {boneName=dyn_legs_flap_01_l_MSTR}
   }
   n52 = Anim Dynamics - Bone: dyn_legs_flap_02_r(ComponentPose=@n51, Alpha=1, ExternalForce="0,0,0") {
    angularDampingOverride=0.999
    angularSpringConstant=800
    bAngularSpring=true
    bGravityOverrideInSimSpace=true
    bOverrideAngularDamping=true
    gravityScale=0.2
    simulationSpace=RootRelative
    alphaBoolBlend {blendOption=Linear}
    boundBone {boneName=dyn_legs_flap_02_r}
    componentAppliedLinearAccClamp=<10000 10000 10000>
    componentLinearAccScale=<0.1 0.1 0.1>
    componentLinearVelScale=<0.1 0.1 0.1>
    gravityOverride {z=-980}
    physicsBodyDefinitions [{
      sphereCollisionRadius=0
      boundBone {boneName=dyn_legs_flap_02_r}
      boxExtents=<5 14 14>
      constraintSetup {angularTargetAxis=AxisY, twistAxis=AxisZ, angularLimitsMax=<5 5 10>, angularLimitsMin=<-5 -30 -10>, angularTarget=<0 0 0>}
      localJointOffset=<-5 -1 0>
     }]
    relativeSpaceBone {boneName=dyn_legs_flap_01_l_MSTR}
   }
   n53 = Anim Dynamics - Bone: dyn_legs_flap_03_r(ComponentPose=@n52, Alpha=1, ExternalForce="0,0,0") {
    angularDampingOverride=0.999
    angularSpringConstant=800
    bAngularSpring=true
    bGravityOverrideInSimSpace=true
    bOverrideAngularDamping=true
    gravityScale=0.2
    simulationSpace=RootRelative
    alphaBoolBlend {blendOption=Linear}
    boundBone {boneName=dyn_legs_flap_03_r}
    componentAppliedLinearAccClamp=<10000 10000 10000>
    componentLinearAccScale=<0.1 0.1 0.1>
    componentLinearVelScale=<0.1 0.1 0.1>
    gravityOverride {z=-980}
    physicsBodyDefinitions [{
      sphereCollisionRadius=0
      boundBone {boneName=dyn_legs_flap_03_r}
      boxExtents=<5 14 14>
      constraintSetup {angularTargetAxis=AxisY, twistAxis=AxisZ, angularLimitsMax=<5 5 10>, angularLimitsMin=<-5 -30 -10>, angularTarget=<0 0 0>}
      localJointOffset=<-5 -1 0>
     }]
    relativeSpaceBone {boneName=dyn_legs_flap_01_l_MSTR}
   }
   Save cached pose 'cachedPose'(Pose=Component To Local(ComponentPose=@n53)) {cacheName=cachedPose}
  ]}
 EventGraph {nodes [
   Event BlueprintUpdateAnimation()
   n1 = OR Boolean(A=Get bCanPlayCustomAnimations() -> (bCanPlayCustomAnimations:bool), B=Get bIsPlayingEmote())
   > Set isEmotingAlpha(isEmotingAlpha:double="To Float (Boolean)"(InBool=@n1)) -> (Output_Get:double)
   > Set isEmotingAlphaInv(isEmotingAlphaInv:double="To Float (Boolean)"(InBool=NOT Boolean(A=@n1))) -> (Output_Get:double)
   n2 = Get CurrentLOD()
   n3 = Get bRigidBodyEnabled()
   > Set isMobileLOD(isMobileLOD:bool=OR Boolean(A=NOT Boolean(A=@n3), B="integer > integer"(A=@n2, B=1) {is=title K2Node_CallFunction})) -> (Output_Get:bool)
   n4 = Get Skel_Mesh()
   > Set thigh_shape_fn_l(thigh_shape_fn_l:double=GetAlphaFromReferenceTransformDelta(SkeletalMeshComponent=@n4, BoneName=thigh_l, BaseName=pelvis, InputRange=(X=-15,Y=-140), TwistAxis="0,0,1")) -> (Output_Get:double)
   > Set thigh_shape_fn_r(thigh_shape_fn_r:double=GetAlphaFromReferenceTransformDelta(SkeletalMeshComponent=@n4, BoneName=thigh_r, BaseName=pelvis, InputRange=(X=-15,Y=-140), TwistAxis="0,0,1")) -> (Output_Get:double)
   Event OnBeginSkydiving()
   > Set isFloatingAlpha(isFloatingAlpha:double=1) -> (Output_Get:double)
   Event OnEndSkydiving()
   > Set isFloatingAlpha(isFloatingAlpha=0)
   Event OnBeginCrouching()
   > Set isCrouchingAlpha(isCrouchingAlpha:double=1) -> (Output_Get:double)
   Event OnEndCrouching()
   > Set isCrouchingAlpha(isCrouchingAlpha=0)
   Event BlueprintInitializeAnimation()
   > Set Skel_Mesh(Skel_Mesh=GetOwnerSkeletalMeshComponent())
  ]}
]
variables [
 isCrouchingAlpha {type=number}
 isEmotingAlpha {type=number}
 isEmotingAlphaInv {type=number}
 isFloatingAlpha {type=number}
 isMobileLOD {type=boolean}
 Skel_Mesh {type=SkeletalMeshComponent}
 thigh_shape_fn_l {type=number}
 thigh_shape_fn_r {type=number}
]
