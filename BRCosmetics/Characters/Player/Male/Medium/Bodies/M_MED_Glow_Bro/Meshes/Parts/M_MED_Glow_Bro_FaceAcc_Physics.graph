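# Physics asset for the Glow Bro medium male FaceAcc part, setting up secondary motion for a ponytail and top hair.
# Tiny MSTR anchor bodies join hair roots to neck_02 with free angular constraints.
# A five-capsule ponytail chain and hairtop boxes 01, 02 and 06 hang off anchors with limited swing and tight twist; hairtops 03, 05 and 07 have only anchors.
# Neck, hoodie and upper-arm capsules join spine_05 with free rotation.
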PhysicsAsset /BRCosmetics/Characters/Player/Male/Medium/Bodies/M_MED_Glow_Bro/Meshes/Parts/M_MED_Glow_Bro_FaceAcc_Physics
bodies {
 dyn_hairtop_01 [{shapeType=Box, extentX=2.5, extentY=2, extentZ=1, center=<0.3893 -0.0001 0.0001>}]
 dyn_hairtop_01_MSTR [{shapeType=Capsule, length=0.01, radius=0.01}]
 dyn_hairtop_02 [{shapeType=Box, extentX=2.5, extentY=2, extentZ=1, center=<0.3893 -0.0001 0.0001>}]
 dyn_hairtop_02_MSTR [{shapeType=Capsule, length=0.01, radius=0.01}]
 dyn_hairtop_03_MSTR [{shapeType=Capsule, length=0.01, radius=0.01}]
 dyn_hairtop_05_MSTR [{shapeType=Capsule, length=0.01, radius=0.01}]
 dyn_hairtop_06 [{shapeType=Box, extentX=2.5, extentY=2, extentZ=1, center=<0.3893 -0.0001 0.0001>}]
 dyn_hairtop_06_MSTR [{shapeType=Capsule, length=0.01, radius=0.01}]
 dyn_hairtop_07_MSTR [{shapeType=Capsule, length=0.01, radius=0.01}]
 dyn_hoodie_01 [{shapeType=Capsule, length=6.75, radius=11, center=<-8.2515 -0.0001 -3.3578>, rotation=<rot 0 0 90>}]
 dyn_ponytail_01 [{shapeType=Capsule, length=2.4682, radius=1.7728, center=<2.2466 0.1581 0.1511>, rotation=<rot 0 90 -90>}]
 dyn_ponytail_01_MSTR [{shapeType=Capsule, length=0.01, radius=0.01}]
 dyn_ponytail_02 [{shapeType=Capsule, length=2.4682, radius=1.7728, center=<2.2466 0.1581 0.1511>, rotation=<rot 0 90 -90>}]
 dyn_ponytail_03 [{shapeType=Capsule, length=2.4682, radius=1.7728, center=<2.2466 0.1581 0.1511>, rotation=<rot 0 90 -90>}]
 dyn_ponytail_04 [{shapeType=Capsule, length=2.4682, radius=1.7728, center=<2.2466 0.1581 0.1511>, rotation=<rot 0 90 -90>}]
 dyn_ponytail_05 [{shapeType=Capsule, length=2.4682, radius=1.7728, center=<2.2466 0.1581 0.1511>, rotation=<rot 0 90 -90>}]
 neck_02 [{shapeType=Capsule, length=14, radius=7.5, center=<-20.2242 25.9062 -0.0002>, rotation=<rot 90 0 -40.0013>}]
 spine_05 [{shapeType=Capsule, length=39, radius=13.5, center=<-1.3174 -0.0001 0>}]
 upperarm_l [{shapeType=Capsule, length=16.483, radius=8.5204, center=<8.0162 0.043 -0.7825>, rotation=<rot 0 90.3029 -89.9999>}]
 upperarm_r [{shapeType=Capsule, length=16.483, radius=8.5204, center=<-8.0162 -0.043 0.7825>, rotation=<rot 0 90.3029 90>}]
}
constraints [
 {
  bone1Name=dyn_ponytail_01_MSTR
  bone2Name=neck_02
  swing1LimitDegrees=45
  swing1Motion=Free
  swing2LimitDegrees=45
  swing2Motion=Free
  twistLimitDegrees=45
  twistMotion=Free
 }
 {
  bone1Name=dyn_hairtop_02_MSTR
  bone2Name=neck_02
  swing1LimitDegrees=45
  swing1Motion=Free
  swing2LimitDegrees=45
  swing2Motion=Free
  twistLimitDegrees=45
  twistMotion=Free
 }
 {
  bone1Name=dyn_hairtop_01_MSTR
  bone2Name=neck_02
  swing1LimitDegrees=45
  swing1Motion=Free
  swing2LimitDegrees=45
  swing2Motion=Free
  twistLimitDegrees=45
  twistMotion=Free
 }
 {
  bone1Name=dyn_hairtop_06_MSTR
  bone2Name=neck_02
  swing1LimitDegrees=45
  swing1Motion=Free
  swing2LimitDegrees=45
  swing2Motion=Free
  twistLimitDegrees=45
  twistMotion=Free
 }
 {
  bone1Name=dyn_hairtop_07_MSTR
  bone2Name=neck_02
  swing1LimitDegrees=45
  swing1Motion=Free
  swing2LimitDegrees=45
  swing2Motion=Free
  twistLimitDegrees=45
  twistMotion=Free
 }
 {
  bone1Name=dyn_hairtop_03_MSTR
  bone2Name=neck_02
  swing1LimitDegrees=45
  swing1Motion=Free
  swing2LimitDegrees=45
  swing2Motion=Free
  twistLimitDegrees=45
  twistMotion=Free
 }
 {
  bone1Name=dyn_hairtop_05_MSTR
  bone2Name=neck_02
  swing1LimitDegrees=45
  swing1Motion=Free
  swing2LimitDegrees=45
  swing2Motion=Free
  twistLimitDegrees=45
  twistMotion=Free
 }
 {bone1Name=dyn_ponytail_01, bone2Name=dyn_ponytail_01_MSTR, swing1LimitDegrees=15, swing2LimitDegrees=15, twistLimitDegrees=5}
 {bone1Name=dyn_ponytail_02, bone2Name=dyn_ponytail_01, swing1LimitDegrees=45, swing2LimitDegrees=45, twistLimitDegrees=5}
 {bone1Name=dyn_ponytail_03, bone2Name=dyn_ponytail_02, swing1LimitDegrees=45, swing2LimitDegrees=45, twistLimitDegrees=5}
 {bone1Name=dyn_ponytail_04, bone2Name=dyn_ponytail_03, swing1LimitDegrees=45, swing2LimitDegrees=45, twistLimitDegrees=5}
 {bone1Name=dyn_ponytail_05, bone2Name=dyn_ponytail_04, swing1LimitDegrees=45, swing2LimitDegrees=45, twistLimitDegrees=5}
 {bone1Name=dyn_hairtop_02, bone2Name=dyn_hairtop_02_MSTR, swing1LimitDegrees=15, swing2LimitDegrees=15, twistLimitDegrees=5}
 {bone1Name=dyn_hairtop_01, bone2Name=dyn_hairtop_01_MSTR, swing1LimitDegrees=15, swing2LimitDegrees=15, twistLimitDegrees=5}
 {bone1Name=dyn_hairtop_06, bone2Name=dyn_hairtop_06_MSTR, swing1LimitDegrees=15, swing2LimitDegrees=15, twistLimitDegrees=5}
 {
  bone1Name=neck_02
  bone2Name=spine_05
  swing1LimitDegrees=45
  swing1Motion=Free
  swing2LimitDegrees=45
  swing2Motion=Free
  twistLimitDegrees=45
  twistMotion=Free
 }
 {
  bone1Name=dyn_hoodie_01
  bone2Name=spine_05
  swing1LimitDegrees=45
  swing1Motion=Free
  swing2LimitDegrees=45
  swing2Motion=Free
  twistLimitDegrees=45
  twistMotion=Free
 }
 {
  bone1Name=upperarm_l
  bone2Name=spine_05
  swing1LimitDegrees=45
  swing1Motion=Free
  swing2LimitDegrees=45
  swing2Motion=Free
  twistLimitDegrees=45
  twistMotion=Free
 }
 {
  bone1Name=upperarm_r
  bone2Name=spine_05
  swing1LimitDegrees=45
  swing1Motion=Free
  swing2LimitDegrees=45
  swing2Motion=Free
  twistLimitDegrees=45
  twistMotion=Free
 }
]
