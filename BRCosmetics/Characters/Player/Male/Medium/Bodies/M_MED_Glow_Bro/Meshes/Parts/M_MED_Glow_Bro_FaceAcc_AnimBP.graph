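# Post-process style Anim Blueprint for Glow Bro's face accessory mesh, animating a dynamic ponytail over a pose copied from the attached parent.
# Normally, RigidBody simulates relative to neck_02, after an aim-pitch-driven bone tilt, with halved skydiving force as wind.
# Above LOD 1 or with rigid bodies disabled, bone constraints to spine_05 plus fixed offsets replace physics.
# A crouch-walk alpha is computed but unused in the AnimGraph.
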AnimBlueprint /BRCosmetics/Characters/Player/Male/Medium/Bodies/M_MED_Glow_Bro/Meshes/Parts/M_MED_Glow_Bro_FaceAcc_AnimBP
target_skeleton=&/M_MED_Glow_Bro_FaceAcc_Skeleton
graphs [
 AnimGraph {nodes [
   n1 = "Transform (Modify) Bone - Bone: dyn_ponytail_01_MSTR"(
    ComponentPose=Local To Component(LocalPose=Use cached pose 'copyPoseCache'())
    Alpha=Get aimUpAdjustAlpha() -> (aimUpAdjustAlpha:double)
    Translation="0,0,0.5"
    Rotation="0,0,45"
    Scale="1,1,1"
   ) {
    rotationMode=BMM_Additive
    translationMode=BMM_Additive
    alphaBoolBlend {blendOption=Linear}
    boneToModify {boneName=dyn_ponytail_01_MSTR}
    rotation {roll=45}
    translation {z=0.5}
   }
   n2 = RigidBody(ComponentPose=@n1, ExternalForce=Get windVector() -> (windVector:Vector)) {
    bClampLinearTranslationLimitToRefPose=true
    bOverrideWorldGravity=true
    overridePhysicsAsset=&/M_MED_Glow_Bro_FaceAcc_Physics
    simulationSpace=BaseBoneSpace
    alphaBoolBlend {blendOption=Linear}
    baseBoneRef {boneName=neck_02}
    componentAppliedLinearAccClamp=<3000 3000 3000>
    componentLinearAccScale=<0.35 0.3 0.3>
    componentLinearVelScale=<0.35 0.3 0.3>
    overrideWorldGravity {z=-980}
   }
   n3 = "Constraint - dyn_ponytail_02 (PP)"!(ComponentPose=Local To Component(LocalPose=Use cached pose 'copyPoseCache'()), Alpha=1, ConstraintWeights_0=1, ConstraintWeights_1:float=0) -> (Pose) {
    alphaBoolBlend {blendOption=Linear}
    boneToModify {boneName=dyn_ponytail_02}
    constraintSetup [{targetBone {boneName=spine_05}}, {targetBone {boneName=dyn_ponytail_01_MSTR}}]
    constraintWeights [1,1]
   }
   n4 = "Constraint - dyn_ponytail_03 (P)"(ComponentPose=@n3, Alpha=1, ConstraintWeights_0=1) {
    alphaBoolBlend {blendOption=Linear}
    boneToModify {boneName=dyn_ponytail_03}
    constraintSetup [{targetBone {boneName=spine_05}}]
    constraintWeights [1]
   }
   n5 = "Constraint - dyn_ponytail_04 (P)"(ComponentPose=@n4, Alpha=1, ConstraintWeights_0=1) {
    alphaBoolBlend {blendOption=Linear}
    boneToModify {boneName=dyn_ponytail_04}
    constraintSetup [{targetBone {boneName=spine_05}}]
    constraintWeights [1]
   }
   n6 = "Constraint - dyn_ponytail_05 (P)"(ComponentPose=@n5, Alpha=1, ConstraintWeights_0=1) {
    alphaBoolBlend {blendOption=Linear}
    boneToModify {boneName=dyn_ponytail_05}
    constraintSetup [{targetBone {boneName=spine_05}}]
    constraintWeights [1]
   }
   n7 = "Transform (Modify) Bone - Bone: dyn_ponytail_02"(ComponentPose=@n6, Alpha=1, Translation="0,0,0.5", Rotation="0,35,0", Scale="1,1,1") {
    rotationMode=BMM_Additive
    rotationSpace=BCS_ParentBoneSpace
    translationMode=BMM_Additive
    translationSpace=BCS_ParentBoneSpace
    alphaBoolBlend {blendOption=Linear}
    boneToModify {boneName=dyn_ponytail_02}
    rotation {yaw=35}
    translation {z=0.5}
   }
   n8 = "Transform (Modify) Bone - Bone: dyn_ponytail_03"(ComponentPose=@n7, Alpha=0, Translation="0,0,0.5", Rotation="0,5,0", Scale="1,1,1") {
    rotationMode=BMM_Additive
    rotationSpace=BCS_ParentBoneSpace
    translationMode=BMM_Additive
    translationSpace=BCS_ParentBoneSpace
    alphaBoolBlend {blendOption=Linear}
    boneToModify {boneName=dyn_ponytail_03}
    rotation {yaw=5}
    translation {z=0.5}
   }
   n9 = Blend Poses by bool(
    BlendPose_0=Component To Local(ComponentPose=@n8)
    BlendPose_1=Component To Local(ComponentPose=@n2)
    bActiveValue=Get isMobileLOD() -> (isMobileLOD:bool)
   ) {blendType=Linear}
   Output Pose(Result=Component To Local(ComponentPose=Local To Component(LocalPose=@n9))) {layerGroup=None}
   Save cached pose 'copyPoseCache'(Pose=Copy Pose From Mesh() {bUseAttachedParent=true}) {cacheName=copyPoseCache}
  ]}
 EventGraph {nodes [
   Event BlueprintUpdateAnimation()
   > n1 = Sequence!(execute) -> (then_0, then_1, then_2:Exec)
   n2 = Get bIsCrouching()
   > Set isCrouchingAndWalkingAlpha(isCrouchingAndWalkingAlpha:double=MapRangeClamped(Value=Get PawnSpeed2D(), InRangeA=0, InRangeB=50, OutRangeB="To Float (Boolean)"(InBool=@n2))) -> (Output_Get:double)
   n3 = Get CurrentLOD()
   n4 = Get bRigidBodyEnabled()
   Set isMobileLOD(execute=@n1.then_1, isMobileLOD:bool=OR Boolean(A=NOT Boolean(A=@n4), B="integer > integer"(A=@n3, B=1) {is=title K2Node_CallFunction})) -> (Output_Get:bool)
   Set aimUpAdjustAlpha(execute=@n1.then_2, aimUpAdjustAlpha:double=MapRangeClamped(Value=Get AimPitch(), InRangeA=0, InRangeB=65)) -> (Output_Get:double)
   n5 = Get SkydivingExternalForceInput()
   > Set windVector(windVector:Vector=vector * float(
     A=Break Fort Anim Input Skydiving External Force(FortAnimInput_SkydivingExternalForce=@n5) -> (
      bUseSkydivingVectorForce=~
      SkydivingVectorForce=~
      HeadToPelvisDirection=~
      FloatingMultiplier=~
      FloatingAdditive=~
      DivingMultiplier=~
      DivingAdditive=~
      ParachutingMultiplier=~
      ParachutingAdditive=~
     ).SkydivingVectorForce
     B=0.5
    )) -> (Output_Get:Vector)
  ]}
]
variables [
 aimUpAdjustAlpha {type=number}
 isCrouchingAndWalkingAlpha {type=number}
 isMobileLOD {type=boolean}
 windVector {type=Vector}
]
